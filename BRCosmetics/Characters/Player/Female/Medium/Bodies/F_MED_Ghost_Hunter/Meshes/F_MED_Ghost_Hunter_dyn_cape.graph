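# Physics asset for the dynamic cape on the Female Medium Ghost Hunter body.
# Left, middle and right capsule chains hang from a back-cape master on spine_05, with tight swing and twist limits per link.
# Cross constraints tie both side-chain tips to the middle chain's end, linking the strips.
# Capsules on the spine, pelvis, legs, upper arms, bag and pistol add body volumes around the cape.
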PhysicsAsset /BRCosmetics/Characters/Player/Female/Medium/Bodies/F_MED_Ghost_Hunter/Meshes/F_MED_Ghost_Hunter_dyn_cape
bodies {
 calf_l [{shapeType=Capsule, length=37.1458, radius=9.7567, center=<-26.7657 -2.507 -0.6294>, rotation=<rot 90 0 -3.6258>}]
 calf_r [{shapeType=Capsule, length=37.1458, radius=9.7567, center=<27.4315 3.5821 0.6294>, rotation=<rot 90 0 177.8577>}]
 dyn_back_cape_MSTR [{shapeType=Capsule, length=0.505, radius=0.505, shapeName=dyn_cap_a_r_MSTR_capsule, rotation=<rot 0 -90 0>}]
 dyn_bag_r_MSTR [{shapeType=Capsule, length=9.1792, radius=6.5511, center=<3.2877 0.6592 0.4689>, rotation=<rot 90 0 0>}]
 dyn_cap_01_mid [{shapeType=Capsule, length=14.75, radius=11.25, shapeName=dyn_cap_a_01_l_capsule, center=<5 0 8>, rotation=<rot 0 0 90>}]
 dyn_cap_02_mid [{shapeType=Capsule, length=14.75, radius=11.25, shapeName=dyn_cap_a_02_l_capsule, center=<8 0 8>, rotation=<rot 0 0 90>}]
 dyn_cap_03_mid [{shapeType=Capsule, length=14.75, radius=11.25, shapeName=dyn_cap_a_03_l_capsule, center=<8 0 8>, rotation=<rot 0 0 90>}]
 dyn_cap_04_mid [{shapeType=Capsule, length=12.2608, radius=11, shapeName=dyn_cap_a_04_l_capsule, center=<15 0 7>, rotation=<rot 0 0 90>}]
 dyn_cap_a_01_l [{shapeType=Capsule, length=8.5014, radius=9.9151, shapeName=dyn_cap_a_01_l_capsule, center=<6.5 0 6>, rotation=<rot 0 0 90>}]
 dyn_cap_a_01_r [{shapeType=Capsule, length=8.4166, radius=9.5, shapeName=dyn_cap_a_01_l_capsule, center=<6.5 0 6>, rotation=<rot 0 0 90>}]
 dyn_cap_a_02_l [{shapeType=Capsule, length=6.7514, radius=10.4151, shapeName=dyn_cap_a_02_l_capsule, center=<8 0 6>, rotation=<rot 0 0 90>}]
 dyn_cap_a_02_r [{shapeType=Capsule, length=7.1666, radius=10, shapeName=dyn_cap_a_02_l_capsule, center=<8 0 6>, rotation=<rot 0 0 90>}]
 dyn_cap_a_03_l [{shapeType=Capsule, length=16.5, radius=10.5, shapeName=dyn_cap_a_03_l_capsule, center=<10 0 6>, rotation=<rot 90 0 0>}]
 dyn_cap_a_03_r [{shapeType=Capsule, length=16.5, radius=10.5, shapeName=dyn_cap_a_03_l_capsule, center=<10 0 6>, rotation=<rot 90 0 0>}]
 dyn_cap_a_l_MSTR [{shapeType=Capsule, length=0.505, radius=0.505, shapeName=dyn_cap_a_l_MSTR_capsule, rotation=<rot 0 -90 0>}]
 dyn_cap_a_r_MSTR [{shapeType=Capsule, length=0.505, radius=0.505, shapeName=dyn_cap_a_r_MSTR_capsule, rotation=<rot 0 -90 0>}]
 dyn_cap_mid_MSTR [{shapeType=Capsule, length=0.505, radius=0.505, shapeName=dyn_cap_mid_MSTR_capsule, rotation=<rot 0 -90 0>}]
 dyn_pistol [{shapeType=Capsule, length=9.2028, radius=9.4937, center=<-2.3172 -3.7478 -1.0518>, rotation=<rot 90 0 -8.378>}]
 pelvis [{shapeType=Capsule, length=13.2434, radius=12.5426, center=<-3.306 -0.2236 1.4124>}]
 spine_02 [{shapeType=Capsule, length=3.7293, radius=9.0328, center=<-0.4061 0.354 0.0003>}]
 spine_04 [{shapeType=Capsule, length=5.1969, radius=10.628, center=<-3.4316 0.7282 0.0001>}]
 spine_05 [{shapeType=Capsule, length=14.9804, radius=8.9968, center=<0 0.3213 0.0001>}]
 thigh_l [{shapeType=Capsule, length=28.6321, radius=10.3895, center=<-21.2757 -0.2297 0.0001>, rotation=<rot 90 90 94.2304>}]
 thigh_r [{shapeType=Capsule, length=28.6321, radius=10.3895, center=<21.2757 0.2297 -0.0001>, rotation=<rot 90 -12.9074 171.323>}]
 upperarm_l [{shapeType=Capsule, length=17.6906, radius=4.6633, shapeName=upperarm_r_capsule, center=<23.518 1.1263 0>, rotation=<rot 90 0 180>}]
 upperarm_r [{shapeType=Capsule, length=18.6906, radius=4.9133, shapeName=upperarm_r_capsule, center=<-23.518 -1.1263 0>, rotation=<rot 90 0 0>}]
}
constraints [
 {bone1Name=spine_05, bone2Name=spine_04, swing1LimitDegrees=45, swing2LimitDegrees=45, twistLimitDegrees=45}
 {bone1Name=thigh_l, bone2Name=pelvis, swing1LimitDegrees=45, swing2LimitDegrees=45, twistLimitDegrees=45}
 {bone1Name=calf_l, bone2Name=thigh_l, swing1LimitDegrees=45, swing2LimitDegrees=45, twistLimitDegrees=45}
 {
  bone1Name=spine_04
  bone2Name=spine_02
  swing1LimitDegrees=45
  swing1Motion=Free
  swing2LimitDegrees=45
  swing2Motion=Free
  twistLimitDegrees=45
  twistMotion=Free
 }
 {
  bone1Name=spine_02
  bone2Name=pelvis
  swing1LimitDegrees=45
  swing1Motion=Free
  swing2LimitDegrees=45
  swing2Motion=Free
  twistLimitDegrees=45
  twistMotion=Free
 }
 {bone1Name=calf_r, bone2Name=thigh_r, swing1LimitDegrees=45, swing2LimitDegrees=45, twistLimitDegrees=45}
 {
  bone1Name=dyn_cap_a_l_MSTR
  bone2Name=dyn_back_cape_MSTR
  swing1LimitDegrees=45
  swing1Motion=Free
  swing2LimitDegrees=45
  swing2Motion=Free
  twistLimitDegrees=45
  twistMotion=Free
 }
 {bone1Name=dyn_cap_a_01_l, bone2Name=dyn_cap_a_l_MSTR, swing1LimitDegrees=15, swing2LimitDegrees=20, twistLimitDegrees=2}
 {bone1Name=dyn_cap_a_02_l, bone2Name=dyn_cap_a_01_l, swing1LimitDegrees=15, swing2LimitDegrees=20, twistLimitDegrees=2}
 {bone1Name=dyn_cap_a_03_l, bone2Name=dyn_cap_a_02_l, swing1LimitDegrees=15, swing2LimitDegrees=20, twistLimitDegrees=2}
 {
  bone1Name=dyn_cap_mid_MSTR
  bone2Name=dyn_back_cape_MSTR
  swing1LimitDegrees=45
  swing1Motion=Free
  swing2LimitDegrees=45
  swing2Motion=Free
  twistLimitDegrees=45
  twistMotion=Free
 }
 {bone1Name=dyn_cap_01_mid, bone2Name=dyn_cap_mid_MSTR, swing1LimitDegrees=15, swing2LimitDegrees=20, twistLimitDegrees=2}
 {bone1Name=dyn_cap_02_mid, bone2Name=dyn_cap_01_mid, swing1LimitDegrees=15, swing2LimitDegrees=20, twistLimitDegrees=2}
 {bone1Name=dyn_cap_03_mid, bone2Name=dyn_cap_02_mid, swing1LimitDegrees=15, swing2LimitDegrees=20, twistLimitDegrees=2}
 {bone1Name=dyn_cap_04_mid, bone2Name=dyn_cap_03_mid, swing1LimitDegrees=15, swing2LimitDegrees=20, twistLimitDegrees=2}
 {
  bone1Name=dyn_cap_a_r_MSTR
  bone2Name=dyn_back_cape_MSTR
  swing1LimitDegrees=45
  swing1Motion=Free
  swing2LimitDegrees=45
  swing2Motion=Free
  twistLimitDegrees=45
  twistMotion=Free
 }
 {bone1Name=dyn_cap_a_01_r, bone2Name=dyn_cap_a_r_MSTR, swing1LimitDegrees=15, swing2LimitDegrees=20, twistLimitDegrees=2}
 {bone1Name=dyn_cap_a_02_r, bone2Name=dyn_cap_a_01_r, swing1LimitDegrees=15, swing2LimitDegrees=20, twistLimitDegrees=2}
 {bone1Name=dyn_cap_a_03_r, bone2Name=dyn_cap_a_02_r, swing1LimitDegrees=15, swing2LimitDegrees=20, twistLimitDegrees=2}
 {
  bone1Name=dyn_bag_r_MSTR
  bone2Name=pelvis
  swing1LimitDegrees=45
  swing1Motion=Free
  swing2LimitDegrees=45
  swing2Motion=Free
  twistLimitDegrees=45
  twistMotion=Free
 }
 {
  bone1Name=upperarm_r
  bone2Name=spine_05
  swing1LimitDegrees=45
  swing1Motion=Free
  swing2LimitDegrees=45
  swing2Motion=Free
  twistLimitDegrees=45
  twistMotion=Free
 }
 {
  bone1Name=dyn_pistol
  bone2Name=thigh_l
  swing1LimitDegrees=45
  swing1Motion=Free
  swing2LimitDegrees=45
  swing2Motion=Free
  twistLimitDegrees=45
  twistMotion=Free
 }
 {
  bone1Name=upperarm_l
  bone2Name=spine_05
  swing1LimitDegrees=45
  swing1Motion=Free
  swing2LimitDegrees=45
  swing2Motion=Free
  twistLimitDegrees=45
  twistMotion=Free
 }
 {bone1Name=dyn_cap_a_03_l, bone2Name=dyn_cap_04_mid, swing1LimitDegrees=15, swing2LimitDegrees=15, twistLimitDegrees=45, twistMotion=Free}
 {bone1Name=dyn_cap_a_03_r, bone2Name=dyn_cap_04_mid, swing1LimitDegrees=15, swing2LimitDegrees=15, twistLimitDegrees=45, twistMotion=Free}
 {bone1Name=thigh_r, bone2Name=pelvis, swing1LimitDegrees=45, swing2LimitDegrees=45, twistLimitDegrees=45}
 {
  bone1Name=dyn_back_cape_MSTR
  bone2Name=spine_05
  swing1LimitDegrees=45
  swing1Motion=Free
  swing2LimitDegrees=45
  swing2Motion=Free
  twistLimitDegrees=45
  twistMotion=Free
 }
]
