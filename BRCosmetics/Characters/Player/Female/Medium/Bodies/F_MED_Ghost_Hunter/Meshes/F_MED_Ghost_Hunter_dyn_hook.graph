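# Physics asset for the Ghost Hunter medium female body cosmetic, defining simulation shapes for its back, left and right "dyn_hook" bones.
# Each hook has a capsule body constrained to its own tiny MSTR anchor capsule, with no constraints between hooks.
# Identical limits give a wide second swing axis but tight first swing and twist, so each hook moves almost like a hinge.
# Two overlapping, unconstrained head capsules approximate the head's volume.
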PhysicsAsset /BRCosmetics/Characters/Player/Female/Medium/Bodies/F_MED_Ghost_Hunter/Meshes/F_MED_Ghost_Hunter_dyn_hook
bodies {
 dyn_hook_bk [{shapeType=Capsule, length=3.4932, radius=5.5752, shapeName=dyn_hook_bk_capsule, center=<6 0 -1.5>, rotation=<rot 0 0 90>}]
 dyn_hook_bk_MSTR [{shapeType=Capsule, length=0.505, radius=0.505, shapeName=dyn_hook_bk_MSTR_capsule, rotation=<rot 0 -90 0>}]
 dyn_hook_l [{shapeType=Capsule, length=4.8559, radius=4.0159, shapeName=dyn_hook_l_capsule, center=<5 0 0>, rotation=<rot 0 0 90>}]
 dyn_hook_l_MSTR [{shapeType=Capsule, length=0.505, radius=0.505, shapeName=dyn_hook_l_MSTR_capsule, rotation=<rot 0 -90 0>}]
 dyn_hook_r [{shapeType=Capsule, length=4.8224, radius=3.927, shapeName=dyn_hook_r_capsule, center=<5 0 0>, rotation=<rot 0 0 90>}]
 dyn_hook_r_MSTR [{shapeType=Capsule, length=0.505, radius=0.505, shapeName=dyn_hook_r_MSTR_capsule, rotation=<rot 0 -90 0>}]
 head [
  {shapeType=Capsule, length=6.125, radius=7.0468, shapeName=head_capsule, center=<-0.6528 0.4511 -0.0001>}
  {shapeType=Capsule, length=3.4642, radius=7.9315, shapeName=head_capsule, center=<4.3775 -1.2237 -0.0001>, rotation=<rot 90 -1.7899 3.6056>}
 ]
}
constraints [
 {bone1Name=dyn_hook_bk, bone2Name=dyn_hook_bk_MSTR, swing1LimitDegrees=10, swing2LimitDegrees=90, twistLimitDegrees=5}
 {bone1Name=dyn_hook_l, bone2Name=dyn_hook_l_MSTR, swing1LimitDegrees=10, swing2LimitDegrees=90, twistLimitDegrees=5}
 {bone1Name=dyn_hook_r, bone2Name=dyn_hook_r_MSTR, swing1LimitDegrees=10, swing2LimitDegrees=90, twistLimitDegrees=5}
]
