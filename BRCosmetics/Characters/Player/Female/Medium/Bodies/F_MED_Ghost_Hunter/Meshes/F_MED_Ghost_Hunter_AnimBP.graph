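# Post-processes the Ghost Hunter cosmetic body, copying the parent mesh pose and adding secondary motion to cape, hooks, pistol and bags.
# Control Rigs, constraints and bone offsets build a cached pre-pose, which feeds RigidBody simulation or fallback trail-controller poses for gameplay and frontend.
# Final offsets adjust for diving, sprinting and parachuting.
# The event graph derives blend alphas from limb rotations, movement, downed and skydiving states, and updates gravity and sim-space values.
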AnimBlueprint /BRCosmetics/Characters/Player/Female/Medium/Bodies/F_MED_Ghost_Hunter/Meshes/F_MED_Ghost_Hunter_AnimBP
target_skeleton=&/F_MED_Ghost_Hunter_Skeleton
graphs [
 AnimGraph {nodes [
   n1 = Get leftCapeConstraintAlpha() -> (leftCapeConstraintAlpha:double)
   n2 = "Constraint - dyn_cap_a_l_MSTR (P)"(ComponentPose=Local To Component(LocalPose=Use cached pose 'prePose'()), Alpha=@n1, ConstraintWeights_0=1) {
    alphaBoolBlend {blendOption=Linear}
    alphaScaleBiasClamp {interpSpeedDecreasing=5}
    boneToModify {boneName=dyn_cap_a_l_MSTR}
    constraintSetup [{targetBone {boneName=upperarm_twist_01_l}}]
    constraintWeights [1]
   }
   n3 = Look At - Bone: dyn_cap_a_l_MSTR(ComponentPose=@n2, Alpha=@n1) {
    alphaBoolBlend {blendOption=Linear}
    boneToModify {boneName=dyn_cap_a_l_MSTR}
    lookAtLocation=<-30 -20 -22>
    lookAtTarget {boneReference {boneName=spine_01}}
    lookAt_Axis {axis=<1 0>}
   }
   n4 = "Constraint - dyn_cap_a_r_MSTR (P)"(ComponentPose=@n3, Alpha=Get rightCapeConstraintAlpha() -> (rightCapeConstraintAlpha:double), ConstraintWeights_0=1) {
    alphaBoolBlend {blendOption=Linear}
    alphaScaleBiasClamp {interpSpeedDecreasing=5}
    boneToModify {boneName=dyn_cap_a_r_MSTR}
    constraintSetup [{targetBone {boneName=upperarm_twist_01_r}}]
    constraintWeights [1]
   }
   Blend Poses by bool(
    BlendPose_0=Component To Local(ComponentPose=@n4)
    BlendPose_1=Component To Local(ComponentPose=Local To Component())
    bActiveValue=Get bIsSprinting()
   ) {blendType=Linear}
   "Constraint - dyn_cap_03_mid (RR)"!(ComponentPose, bAlphaBoolEnabled:bool=True, ConstraintWeights_0=1, ConstraintWeights_1:float=1) -> (Pose) {
    alphaInputType=Bool
    alphaBoolBlend {blendInTime=0.1, blendOption=Linear, blendOutTime=0.1}
    boneToModify {boneName=dyn_cap_03_mid}
    constraintSetup [{transformType=Rotation, targetBone {boneName=calf_twist_01_l}}, {transformType=Rotation, targetBone {boneName=calf_twist_01_r}}]
    constraintWeights [1,1]
   }
   n5 = Get arm_rot_bk_r() -> (arm_rot_bk_r:double)
   n6 = "Transform (Modify) Bone - Bone: dyn_cap_a_01_r"(Alpha=@n5, Translation="0,0,3", Rotation="5,0,0", Scale="1,1,1") {
    rotationMode=BMM_Additive
    rotationSpace=BCS_BoneSpace
    translationMode=BMM_Additive
    translationSpace=BCS_BoneSpace
    alphaBoolBlend {blendOption=Linear}
    boneToModify {boneName=dyn_cap_a_01_r}
    rotation {pitch=5}
    translation {z=3}
   }
   "Transform (Modify) Bone - Bone: deform_cape_arm_r"(ComponentPose=@n6, Alpha=@n5, Translation="0,-4,0", Rotation="0,0,0", Scale="1,1,1") {
    rotationMode=BMM_Additive
    rotationSpace=BCS_BoneSpace
    translationMode=BMM_Additive
    translationSpace=BCS_BoneSpace
    alphaBoolBlend {blendOption=Linear}
    boneToModify {boneName=deform_cape_arm_r}
    translation {y=-4}
   }
   "Transform (Modify) Bone - Bone: dyn_cap_a_l_MSTR"(Alpha=Get arm_rot_bk_l() -> (arm_rot_bk_l:double), Translation="-0.000,-12.942,32.380", Rotation="0,0,0", Scale="1,1,1") {
    rotationMode=BMM_Additive
    rotationSpace=BCS_BoneSpace
    translationMode=BMM_Additive
    translationSpace=BCS_BoneSpace
    alphaBoolBlend {blendOption=Linear}
    boneToModify {boneName=dyn_cap_a_l_MSTR}
    translation {y=-12.942, z=32.38}
   }
   RigidBody!(ComponentPose, Alpha, OverrideWorldGravity:Vector="0,0,0", ExternalForce) -> (Pose) {
    bClampLinearTranslationLimitToRefPose=true
    bOverrideWorldGravity=true
    overridePhysicsAsset=&/F_MED_Ghost_Hunter_dyn_cape
    simulationSpace=BaseBoneSpace
    alphaBoolBlend {blendOption=Linear}
    baseBoneRef {boneName=dyn_root_simspace}
    componentAppliedLinearAccClamp=<3000 3000 3000>
    componentLinearAccScale {z=0.45}
    componentLinearVelScale {z=0.45}
    overrideWorldGravity {z=-980}
    simSpaceSettings {
     maxAngularAcceleration=200
     maxAngularVelocity=3
     maxLinearAcceleration=200
     maxLinearVelocity=100
     velocityScaleZ=0
     worldAlpha=1
     externalLinearDragV=<1 1>
    }
   }
   n7 = Get bIsParachuteOpened()
   n8 = Blend Poses by bool(BlendPose_0=Use cached pose 'frontendPose'(), BlendPose_1=Use cached pose 'trailPose'(), bActiveValue=Get bIsInFrontEnd()) {blendType=Linear}
   n9 = Use cached pose 'rigidPose'()
   n10 = Get bRigidBodyEnabled()
   n11 = Get head_rot_bk() -> (head_rot_bk:double)
   n12 = "Transform (Modify) Bone - Bone: dyn_hook_bk_MSTR"(
    ComponentPose=Local To Component(LocalPose=Blend Poses by bool(BlendPose_0=@n9, BlendPose_1=@n8, bActiveValue=@n10) {blendType=Linear})
    Alpha=@n11
    Translation="0,0,0"
    Rotation="0,0,0"
    Scale="1.5,1,1.000"
   ) {
    rotationMode=BMM_Additive
    rotationSpace=BCS_BoneSpace
    scaleMode=BMM_Additive
    scaleSpace=BCS_BoneSpace
    translationMode=BMM_Additive
    translationSpace=BCS_BoneSpace
    alphaBoolBlend {blendOption=Linear}
    boneToModify {boneName=dyn_hook_bk_MSTR}
    scale {x=1.5}
   }
   n13 = "Transform (Modify) Bone - Bone: dyn_hook_r_MSTR"(ComponentPose=@n12, Alpha=@n11, Translation="0,0,0", Rotation="0,0,0", Scale="1.500,1.000,1.000") {
    rotationMode=BMM_Additive
    rotationSpace=BCS_BoneSpace
    scaleMode=BMM_Additive
    scaleSpace=BCS_BoneSpace
    translationMode=BMM_Additive
    translationSpace=BCS_BoneSpace
    alphaBoolBlend {blendOption=Linear}
    boneToModify {boneName=dyn_hook_r_MSTR}
    scale {x=1.5002}
   }
   n14 = "Transform (Modify) Bone - Bone: dyn_hook_l_MSTR"(ComponentPose=@n13, Alpha=@n11, Translation="0,0,0", Rotation="0,0,0", Scale="1.5,1.000,1.000") {
    rotationMode=BMM_Additive
    rotationSpace=BCS_BoneSpace
    scaleMode=BMM_Additive
    scaleSpace=BCS_BoneSpace
    translationMode=BMM_Additive
    translationSpace=BCS_BoneSpace
    alphaBoolBlend {blendOption=Linear}
    boneToModify {boneName=dyn_hook_l_MSTR}
    scale {x=1.5}
   }
   n15 = Get isDiving() -> (isDiving:double)
   n16 = "Transform (Modify) Bone - Bone: deform_cape_arm_r"(ComponentPose=@n14, Alpha=@n15, Translation="0,-3,0", Rotation="0,0,0", Scale="1,1,1") {
    rotationMode=BMM_Additive
    rotationSpace=BCS_BoneSpace
    translationMode=BMM_Additive
    translationSpace=BCS_BoneSpace
    alphaBoolBlend {blendOption=Linear}
    alphaScaleBiasClamp {bInterpResult=true, interpSpeedDecreasing=5, interpSpeedIncreasing=20}
    boneToModify {boneName=deform_cape_arm_r}
    translation {y=-3}
   }
   n17 = "Transform (Modify) Bone - Bone: deform_cape_arm_l"(ComponentPose=@n16, Alpha=@n15, Translation="0,-3,0", Rotation="0,0,0", Scale="1,1,1") {
    rotationMode=BMM_Additive
    rotationSpace=BCS_BoneSpace
    translationMode=BMM_Additive
    translationSpace=BCS_BoneSpace
    alphaBoolBlend {blendOption=Linear}
    alphaScaleBiasClamp {bInterpResult=true, interpSpeedDecreasing=5, interpSpeedIncreasing=20}
    boneToModify {boneName=deform_cape_arm_l}
    translation {y=-3}
   }
   n18 = Get bIsTacticalSprinting()
   n19 = "Transform (Modify) Bone - Bone: dyn_cap_mid_MSTR"!(ComponentPose=@n17, bAlphaBoolEnabled:bool=@n18, Translation="0,0,5", Rotation="0,0,0", Scale="1,1,1") -> (Pose) {
    alphaInputType=Bool
    rotationMode=BMM_Additive
    rotationSpace=BCS_BoneSpace
    translationMode=BMM_Additive
    translationSpace=BCS_BoneSpace
    alphaBoolBlend {blendOption=Linear}
    alphaScaleBiasClamp {bInterpResult=true, interpSpeedDecreasing=5, interpSpeedIncreasing=20}
    boneToModify {boneName=dyn_cap_mid_MSTR}
    translation {z=5}
   }
   n20 = "Transform (Modify) Bone - Bone: dyn_cap_a_r_MSTR"!(ComponentPose=@n19, bAlphaBoolEnabled:bool=@n18, Translation="0,0,5", Rotation="0,0,0", Scale="1,1,1") -> (Pose) {
    alphaInputType=Bool
    rotationMode=BMM_Additive
    rotationSpace=BCS_BoneSpace
    translationMode=BMM_Additive
    translationSpace=BCS_BoneSpace
    alphaBoolBlend {blendOption=Linear}
    alphaScaleBiasClamp {bInterpResult=true, interpSpeedDecreasing=5, interpSpeedIncreasing=20}
    boneToModify {boneName=dyn_cap_a_r_MSTR}
    translation {z=5}
   }
   n21 = "Transform (Modify) Bone - Bone: deform_cape_arm_r"!(ComponentPose=@n20, bAlphaBoolEnabled:bool=@n18, Translation="0,-5,0", Rotation="0,0,0", Scale="1,1,1") -> (Pose) {
    alphaInputType=Bool
    rotationMode=BMM_Additive
    rotationSpace=BCS_BoneSpace
    translationMode=BMM_Additive
    translationSpace=BCS_BoneSpace
    alphaBoolBlend {blendOption=Linear}
    alphaScaleBiasClamp {bInterpResult=true, interpSpeedDecreasing=5, interpSpeedIncreasing=20}
    boneToModify {boneName=deform_cape_arm_r}
    translation {y=-5}
   }
   n22 = "Transform (Modify) Bone - Bone: dyn_cap_a_l_MSTR"!(ComponentPose=@n21, bAlphaBoolEnabled:bool=@n18, Translation="0,0,5", Rotation="0,0,0", Scale="1,1,1") -> (Pose) {
    alphaInputType=Bool
    rotationMode=BMM_Additive
    rotationSpace=BCS_BoneSpace
    translationMode=BMM_Additive
    translationSpace=BCS_BoneSpace
    alphaBoolBlend {blendOption=Linear}
    alphaScaleBiasClamp {bInterpResult=true, interpSpeedDecreasing=5, interpSpeedIncreasing=20}
    boneToModify {boneName=dyn_cap_a_l_MSTR}
    translation {z=5}
   }
   n23 = "Transform (Modify) Bone - Bone: deform_cape_arm_l"!(ComponentPose=@n22, bAlphaBoolEnabled:bool=@n18, Translation="0,5,0", Rotation="0,0,0", Scale="1,1,1") -> (Pose) {
    alphaInputType=Bool
    rotationMode=BMM_Additive
    rotationSpace=BCS_BoneSpace
    translationMode=BMM_Additive
    translationSpace=BCS_BoneSpace
    alphaBoolBlend {blendOption=Linear}
    alphaScaleBiasClamp {bInterpResult=true, interpSpeedDecreasing=5, interpSpeedIncreasing=20}
    boneToModify {boneName=deform_cape_arm_l}
    translation {y=5}
   }
   n24 = Get bIsSprinting()
   n25 = "Transform (Modify) Bone - Bone: dyn_cap_mid_MSTR"(ComponentPose=@n23, bAlphaBoolEnabled=@n24, Translation="0,0,1", Rotation="0,0,0", Scale="1,1,1") {
    alphaInputType=Bool
    rotationMode=BMM_Additive
    rotationSpace=BCS_BoneSpace
    translationMode=BMM_Additive
    translationSpace=BCS_BoneSpace
    alphaBoolBlend {blendOption=Linear}
    alphaScaleBiasClamp {bInterpResult=true, interpSpeedDecreasing=5, interpSpeedIncreasing=20}
    boneToModify {boneName=dyn_cap_mid_MSTR}
    translation {z=1}
   }
   n26 = "Transform (Modify) Bone - Bone: dyn_cap_a_r_MSTR"(ComponentPose=@n25, bAlphaBoolEnabled=@n24, Translation="0,0,1", Rotation="0,0,0", Scale="1,1,1") {
    alphaInputType=Bool
    rotationMode=BMM_Additive
    rotationSpace=BCS_BoneSpace
    translationMode=BMM_Additive
    translationSpace=BCS_BoneSpace
    alphaBoolBlend {blendOption=Linear}
    alphaScaleBiasClamp {bInterpResult=true, interpSpeedDecreasing=5, interpSpeedIncreasing=20}
    boneToModify {boneName=dyn_cap_a_r_MSTR}
    translation {z=1}
   }
   n27 = "Transform (Modify) Bone - Bone: deform_cape_arm_r"(ComponentPose=@n26, bAlphaBoolEnabled=@n24, Translation="0,-1,0", Rotation="0,0,0", Scale="1,1,1") {
    alphaInputType=Bool
    rotationMode=BMM_Additive
    rotationSpace=BCS_BoneSpace
    translationMode=BMM_Additive
    translationSpace=BCS_BoneSpace
    alphaBoolBlend {blendOption=Linear}
    alphaScaleBiasClamp {bInterpResult=true, interpSpeedDecreasing=5, interpSpeedIncreasing=20}
    boneToModify {boneName=deform_cape_arm_r}
    translation {y=-1}
   }
   n28 = "Transform (Modify) Bone - Bone: dyn_cap_a_l_MSTR"(ComponentPose=@n27, bAlphaBoolEnabled=@n24, Translation="0,0,1", Rotation="0,0,0", Scale="1,1,1") {
    alphaInputType=Bool
    rotationMode=BMM_Additive
    rotationSpace=BCS_BoneSpace
    translationMode=BMM_Additive
    translationSpace=BCS_BoneSpace
    alphaBoolBlend {blendOption=Linear}
    alphaScaleBiasClamp {bInterpResult=true, interpSpeedDecreasing=5, interpSpeedIncreasing=20}
    boneToModify {boneName=dyn_cap_a_l_MSTR}
    translation {z=1}
   }
   n29 = "Transform (Modify) Bone - Bone: deform_cape_arm_l"(ComponentPose=@n28, bAlphaBoolEnabled=@n24, Translation="0,1,0", Rotation="0,0,0", Scale="1,1,1") {
    alphaInputType=Bool
    rotationMode=BMM_Additive
    rotationSpace=BCS_BoneSpace
    translationMode=BMM_Additive
    translationSpace=BCS_BoneSpace
    alphaBoolBlend {blendOption=Linear}
    alphaScaleBiasClamp {bInterpResult=true, interpSpeedDecreasing=5, interpSpeedIncreasing=20}
    boneToModify {boneName=deform_cape_arm_l}
    translation {y=1}
   }
   n30 = "Constraint - dyn_cap_a_03_l (R)"!(ComponentPose=@n29, bAlphaBoolEnabled:bool=@n7, ConstraintWeights_0=1) -> (Pose) {
    alphaInputType=Bool
    alphaBoolBlend {blendInTime=0.1, blendOption=Linear, blendOutTime=0.1}
    boneToModify {boneName=dyn_cap_a_03_l}
    constraintSetup [{transformType=Rotation, targetBone {boneName=calf_l}}]
    constraintWeights [1]
   }
   n31 = "Constraint - dyn_cap_04_mid (RR)"!(ComponentPose=@n30, Alpha=SelectFloat(A=0.6, B=0, bPickA=@n7), ConstraintWeights_0=1, ConstraintWeights_1:float=1) -> (Pose) {
    alphaBoolBlend {blendInTime=0.1, blendOption=Linear, blendOutTime=0.1}
    boneToModify {boneName=dyn_cap_04_mid}
    constraintSetup [{transformType=Rotation, targetBone {boneName=calf_twist_01_l}}, {transformType=Rotation, targetBone {boneName=calf_twist_01_r}}]
    constraintWeights [1,1]
   }
   n32 = "Constraint - dyn_cap_a_03_r (R)"!(ComponentPose=@n31, bAlphaBoolEnabled:bool=@n7, ConstraintWeights_0=1) -> (Pose) {
    alphaInputType=Bool
    alphaBoolBlend {blendInTime=0.1, blendOption=Linear, blendOutTime=0.1}
    boneToModify {boneName=dyn_cap_a_03_r}
    constraintSetup [{transformType=Rotation, targetBone {boneName=calf_r}}]
    constraintWeights [1]
   }
   Output Pose(Result=Component To Local(ComponentPose=@n32)) {layerGroup=None}
   n33 = Control Rig(Source=Copy Pose From Mesh() {bUseAttachedParent=true}) {
    alphaBoolBlend {blendOption=Linear}
    controlRigAssetReference {blueprintRigClass=&/F_MED_Ghost_Hunter_Deform_CtrlRig_twists.F_MED_Ghost_Hunter_Deform_CtrlRig_twists_C}
   }
   n34 = Control Rig!(
    Source=@n33
    Alpha
    KneeRot_fr_l_114deg:Vector
    KneeTrans_fr_l_114deg:Vector="0.258,-3.591,0.009"
    KneeScale_fr_l_114deg:Vector="1,1,1"
    KneeScale_fr_l_130deg:Vector="1,1,1"
    KneeRot_fr_l_130deg:Vector="0,0,-51.2"
    KneeTrans_fr_l_130deg:Vector="0.258,-3.591,0.009"
   ) -> (Pose) {
    alphaBoolBlend {blendOption=Linear}
    controlRigAssetReference {blueprintRigClass=&/F_MED_Ghost_Hunter_Deform_CtrlRig_sculpters.F_MED_Ghost_Hunter_Deform_CtrlRig_sculpters_C}
   }
   n35 = Control Rig!(Source=@n34, Alpha) -> (Pose) {
    alphaBoolBlend {blendOption=Linear}
    controlRigAssetReference {blueprintRigClass=&/F_MED_Ghost_Hunter_Deform_CtrlRig_cape.F_MED_Ghost_Hunter_Deform_CtrlRig_cape_C}
   }
   n36 = "Constraint - dyn_pistol_MSTR (PPP)"!(
    ComponentPose=Local To Component(LocalPose=@n35)
    Alpha=1
    ConstraintWeights_0=0.1721
    ConstraintWeights_1:float=0.5806
    ConstraintWeights_2:float=0.2473
   ) -> (Pose) {
    alphaBoolBlend {blendOption=Linear}
    boneToModify {boneName=dyn_pistol_MSTR}
    constraintSetup [{targetBone {boneName=pelvis}}, {targetBone {boneName=thigh_twist_01_l}}, {targetBone {boneName=thigh_l}}]
    constraintWeights [1,1,1]
   }
   n37 = "Constraint - dyn_bag_r_MSTR (PP)"!(ComponentPose=@n36, Alpha=1, ConstraintWeights_0=0.9, ConstraintWeights_1:float=0.1) -> (Pose) {
    alphaBoolBlend {blendOption=Linear}
    boneToModify {boneName=dyn_bag_r_MSTR}
    constraintSetup [{targetBone {boneName=pelvis}}, {targetBone {boneName=thigh_twist_01_r}}]
    constraintWeights [1,1]
   }
   n38 = "Constraint - dyn_bag_l_01_MSTR (PP)"!(ComponentPose=@n37, Alpha=1, ConstraintWeights_0=0.3215, ConstraintWeights_1:float=0.6785) -> (Pose) {
    alphaBoolBlend {blendOption=Linear}
    boneToModify {boneName=dyn_bag_l_01_MSTR}
    constraintSetup [{targetBone {boneName=spine_01}}, {targetBone {boneName=pelvis}}]
    constraintWeights [1,1]
   }
   n39 = "Constraint - dyn_bag_l_02_MSTR (PP)"!(ComponentPose=@n38, Alpha=1, ConstraintWeights_0=0.3075, ConstraintWeights_1:float=0.6925) -> (Pose) {
    alphaBoolBlend {blendOption=Linear}
    boneToModify {boneName=dyn_bag_l_02_MSTR}
    constraintSetup [{targetBone {boneName=spine_01}}, {targetBone {boneName=pelvis}}]
    constraintWeights [1,1]
   }
   n40 = "Transform (Modify) Bone - Bone: dyn_pistol"(ComponentPose=@n39, Alpha=1, Translation="2.407,0.951,-0.000", Rotation="0.000,0.000,-11.737", Scale="1,1,1") {
    rotationMode=BMM_Additive
    rotationSpace=BCS_BoneSpace
    translationMode=BMM_Additive
    translationSpace=BCS_BoneSpace
    alphaBoolBlend {blendOption=Linear}
    boneToModify {boneName=dyn_pistol}
    rotation {roll=-11.7371}
    translation=<2.407 0.951>
   }
   n41 = Get head_rot_bk()
   n42 = "Transform (Modify) Bone - Bone: dyn_hook_bk_MSTR"(ComponentPose=@n40, Alpha=@n41, Translation="-1.353,-0.000,4.374", Rotation="10.000,0.000,0.000", Scale="1,1,1") {
    rotationMode=BMM_Additive
    rotationSpace=BCS_BoneSpace
    scaleMode=BMM_Additive
    scaleSpace=BCS_BoneSpace
    translationMode=BMM_Additive
    translationSpace=BCS_BoneSpace
    alphaBoolBlend {blendOption=Linear}
    boneToModify {boneName=dyn_hook_bk_MSTR}
    rotation {pitch=10}
    translation {x=-1.353, z=4.374}
   }
   n43 = "Transform (Modify) Bone - Bone: dyn_hook_r_MSTR"(ComponentPose=@n42, Alpha=@n41, Translation="-0.000,0.000,2.095", Rotation="0.000,0.000,5.000", Scale="1,1,1") {
    rotationMode=BMM_Additive
    rotationSpace=BCS_BoneSpace
    scaleMode=BMM_Additive
    scaleSpace=BCS_BoneSpace
    translationMode=BMM_Additive
    translationSpace=BCS_BoneSpace
    alphaBoolBlend {blendOption=Linear}
    boneToModify {boneName=dyn_hook_r_MSTR}
    rotation {roll=5.0001}
    translation {z=2.095}
   }
   n44 = "Transform (Modify) Bone - Bone: dyn_hook_l_MSTR"(ComponentPose=@n43, Alpha=@n41, Translation="0.000,0.000,2.090", Rotation="0.000,0.000,10.000", Scale="1,1,1") {
    rotationMode=BMM_Additive
    rotationSpace=BCS_BoneSpace
    scaleMode=BMM_Additive
    scaleSpace=BCS_BoneSpace
    translationMode=BMM_Additive
    translationSpace=BCS_BoneSpace
    alphaBoolBlend {blendOption=Linear}
    boneToModify {boneName=dyn_hook_l_MSTR}
    rotation {roll=10.0004}
    translation {z=2.09}
   }
   n45 = "Transform (Modify) Bone - Bone: dyn_bag_l_01_MSTR"(
    ComponentPose=@n44
    Alpha=Get thigh_rot_up_l() -> (thigh_rot_up_l:double)
    Translation="-0.034,-1.871,-0.231"
    Rotation="-5.646,-6.603,-19.491"
    Scale="1,1,1"
   ) {
    rotationMode=BMM_Additive
    rotationSpace=BCS_BoneSpace
    translationMode=BMM_Additive
    translationSpace=BCS_BoneSpace
    alphaBoolBlend {blendOption=Linear}
    boneToModify {boneName=dyn_bag_l_01_MSTR}
    rotation=<rot -5.646 -6.603 -19.491>
    translation=<-0.034 -1.871 -0.231>
   }
   n46 = "Transform (Modify) Bone - Bone: dyn_bag_r_MSTR"(ComponentPose=@n45, Alpha=Get thigh_rot_up_l(), Translation="-0.000,1.552,0.000", Rotation="3.696,0.429,6.652", Scale="1,1,1") {
    rotationMode=BMM_Additive
    rotationSpace=BCS_BoneSpace
    translationMode=BMM_Additive
    translationSpace=BCS_BoneSpace
    alphaBoolBlend {blendOption=Linear}
    boneToModify {boneName=dyn_bag_r_MSTR}
    rotation=<rot 3.696 0.429 6.6521>
    translation {y=1.552}
   }
   Save cached pose 'prePose'(Pose=Component To Local(ComponentPose=@n46)) {cacheName=prePose}
   n47 = RigidBody(ComponentPose=Local To Component(LocalPose=Use cached pose 'prePose'()), OverrideWorldGravity=Get GravityOverride()) {
    bClampLinearTranslationLimitToRefPose=true
    bOverrideWorldGravity=true
    overridePhysicsAsset=&F_MED_Ghost_Hunter_dyn_cape
    simulationSpace=BaseBoneSpace
    alphaBoolBlend {blendOption=Linear}
    baseBoneRef {boneName=spine_01}
    componentAppliedLinearAccClamp=<3000 3000 3000>
    componentLinearAccScale {x=0.4}
    componentLinearVelScale {x=0.4}
    overrideWorldGravity {z=-980}
    simSpaceSettings {
     maxAngularAcceleration=100
     maxAngularVelocity=2
     maxLinearAcceleration=500
     maxLinearVelocity=50
     velocityScaleZ=0
     worldAlpha=1
     externalLinearDragV=<1 1 1>
    }
   }
   n48 = RigidBody!(ComponentPose=@n47, Alpha, ExternalForce) -> (Pose) {
    bClampLinearTranslationLimitToRefPose=true
    bOverrideWorldGravity=true
    overridePhysicsAsset=&/F_MED_Ghost_Hunter_dyn_hook
    simulationSpace=BaseBoneSpace
    alphaBoolBlend {blendOption=Linear}
    baseBoneRef {boneName=spine_05}
    componentAppliedLinearAccClamp=<3000 3000 3000>
    componentLinearAccScale=<0.3 0.3 0.3>
    componentLinearVelScale=<0.3 0.3 0.3>
    overrideWorldGravity {z=-980}
    simSpaceSettings {
     maxAngularAcceleration=200
     maxAngularVelocity=3
     maxLinearAcceleration=200
     maxLinearVelocity=100
     velocityScaleZ=0
     externalLinearDragV=<1 1>
    }
   }
   Save cached pose 'rigidPose'(Pose=Component To Local(ComponentPose=@n48)) {cacheName=rigidPose}
   n49 = "Constraint - dyn_back_cape_MSTR (P)"(ComponentPose=Local To Component(LocalPose=Use cached pose 'prePose'()), Alpha=1, ConstraintWeights_0=1) {
    alphaBoolBlend {blendOption=Linear}
    boneToModify {boneName=dyn_back_cape_MSTR}
    constraintSetup [{targetBone {boneName=spine_03}}]
    constraintWeights [1]
   }
   n50 = "Constraint - dyn_cap_a_r_MSTR (PP)"!(ComponentPose=@n49, Alpha=1, ConstraintWeights_0=0.2, ConstraintWeights_1:float=0.8) -> (Pose) {
    alphaBoolBlend {blendOption=Linear}
    boneToModify {boneName=dyn_cap_a_r_MSTR}
    constraintSetup [{targetBone {boneName=spine_03}}, {targetBone {boneName=upperarm_twist_01_r}}]
    constraintWeights [1,1]
   }
   n51 = "Constraint - dyn_cap_a_l_MSTR (PP)"!(ComponentPose=@n50, Alpha=1, ConstraintWeights_0=0.2, ConstraintWeights_1:float=0.8) -> (Pose) {
    alphaBoolBlend {blendOption=Linear}
    boneToModify {boneName=dyn_cap_a_l_MSTR}
    constraintSetup [{targetBone {boneName=spine_03}}, {targetBone {boneName=upperarm_twist_01_l}}]
    constraintWeights [1,1]
   }
   n52 = "Constraint - dyn_cap_a_02_r (PP)"!(ComponentPose=@n51, Alpha=1, ConstraintWeights_0=0.9, ConstraintWeights_1:float=0.1) -> (Pose) {
    alphaBoolBlend {blendOption=Linear}
    boneToModify {boneName=dyn_cap_a_02_r}
    constraintSetup [{targetBone {boneName=spine_03}}, {targetBone {boneName=upperarm_twist_01_r}}]
    constraintWeights [1,1]
   }
   n53 = "Constraint - dyn_cap_a_02_l (PP)"!(ComponentPose=@n52, Alpha=1, ConstraintWeights_0=0.9, ConstraintWeights_1:float=0.1) -> (Pose) {
    alphaBoolBlend {blendOption=Linear}
    boneToModify {boneName=dyn_cap_a_02_l}
    constraintSetup [{targetBone {boneName=spine_03}}, {targetBone {boneName=upperarm_twist_01_l}}]
    constraintWeights [1,1]
   }
   n54 = "Constraint - dyn_cap_02_mid (PP)"!(ComponentPose=@n53, Alpha=1, ConstraintWeights_0=0.8, ConstraintWeights_1:float=0.2) -> (Pose) {
    alphaBoolBlend {blendOption=Linear}
    boneToModify {boneName=dyn_cap_02_mid}
    constraintSetup [{targetBone {boneName=pelvis}}, {targetBone {boneName=spine_02}}]
    constraintWeights [1,1]
   }
   n55 = "Constraint - dyn_cap_a_03_r (PP)"!(ComponentPose=@n54, Alpha=1, ConstraintWeights_0=0.8, ConstraintWeights_1:float=0.2) -> (Pose) {
    alphaBoolBlend {blendOption=Linear}
    boneToModify {boneName=dyn_cap_a_03_r}
    constraintSetup [{targetBone {boneName=pelvis}}, {targetBone {boneName=spine_02}}]
    constraintWeights [1,1]
   }
   n56 = "Constraint - dyn_cap_a_03_l (PP)"!(ComponentPose=@n55, Alpha=1, ConstraintWeights_0=0.8, ConstraintWeights_1:float=0.2) -> (Pose) {
    alphaBoolBlend {blendOption=Linear}
    boneToModify {boneName=dyn_cap_a_03_l}
    constraintSetup [{targetBone {boneName=pelvis}}, {targetBone {boneName=spine_02}}]
    constraintWeights [1,1]
   }
   n57 = Get BwdMovement() -> (BwdMovement:double)
   n58 = Trail controller - Bone: dyn_cap_04_mid!(ComponentPose=@n56, Alpha=@n57, FakeVelocity:Vector="0,0,0") -> (Pose) {
    bActorSpaceFakeVel=true
    bShowTrailLocation=true
    bUsePlanarLimit=true
    chainLength=4
    debugLifeTime=0.1
    maxDeltaTime=0.0333
    alphaScaleBiasClamp {bInterpResult=true}
    baseJoint {boneName=root}
    planarLimits [
     {drivingBone {boneName=root}, planeTransform {location=<0 0 10>, rotation=<rot 0 0 0>, scale=<1 1 1>}}
     {drivingBone {boneName=spine_03}, planeTransform {location=<0 -20 0>, rotation=<rot 0 15 -89.9999>, scale=<1 1 1>}}
    ]
    rotationLimits [{limitMax=<180 180 180>}, {limitMax=<180 180 180>}, {limitMax=<180 180 180>}, {limitMax=<180 180 180>}]
    rotationOffsets [<0 0 0>, <0 0 0>, <0 0 0>, <0 0 0>]
    trailBone {boneName=dyn_cap_04_mid}
    trailRelaxationSpeed {editorCurveData {keys [<at 0 30 RCIM_Linear>,<at 1 10 RCIM_Linear>]}}
   }
   n59 = Trail controller - Bone: dyn_cap_a_03_l!(ComponentPose=@n58, Alpha=@n57, FakeVelocity:Vector="0,0,0") -> (Pose) {
    bActorSpaceFakeVel=true
    bShowTrailLocation=true
    bUsePlanarLimit=true
    debugLifeTime=0.1
    maxDeltaTime=0.0333
    alphaScaleBiasClamp {bInterpResult=true}
    baseJoint {boneName=root}
    planarLimits [
     {drivingBone {boneName=root}, planeTransform {location=<0 0 10>, rotation=<rot 0 0 0>, scale=<1 1 1>}}
     {drivingBone {boneName=spine_03}, planeTransform {location=<0 -20 0>, rotation=<rot 0 15 -89.9999>, scale=<1 1 1>}}
    ]
    rotationLimits [{limitMax=<180 180 180>}, {limitMax=<180 180 180>}]
    rotationOffsets [<0 0 0>, <0 0 0>]
    trailBone {boneName=dyn_cap_a_03_l}
    trailRelaxationSpeed {editorCurveData {keys [<at 0 30 RCIM_Linear>,<at 1 10 RCIM_Linear>]}}
   }
   n60 = Trail controller - Bone: dyn_cap_a_03_r!(ComponentPose=@n59, Alpha=@n57, FakeVelocity:Vector="0,0,0") -> (Pose) {
    bActorSpaceFakeVel=true
    bShowTrailLocation=true
    bUsePlanarLimit=true
    debugLifeTime=0.1
    maxDeltaTime=0.0333
    alphaScaleBiasClamp {bInterpResult=true}
    baseJoint {boneName=root}
    planarLimits [
     {drivingBone {boneName=root}, planeTransform {location=<0 0 10>, rotation=<rot 0 0 0>, scale=<1 1 1>}}
     {drivingBone {boneName=spine_03}, planeTransform {location=<0 -20 0>, rotation=<rot 0 15 -89.9999>, scale=<1 1 1>}}
    ]
    rotationLimits [{limitMax=<180 180 180>}, {limitMax=<180 180 180>}]
    rotationOffsets [<0 0 0>, <0 0 0>]
    trailBone {boneName=dyn_cap_a_03_r}
    trailRelaxationSpeed {editorCurveData {keys [<at 0 30 RCIM_Linear>,<at 1 10 RCIM_Linear>]}}
   }
   n61 = Get DBNO_alpha() -> (DBNO_alpha:bool)
   n62 = "Transform (Modify) Bone - Bone: dyn_cap_01_mid"!(ComponentPose=@n60, bAlphaBoolEnabled:bool=@n61, Translation="0,0,-2", Rotation="0,0,0", Scale="1,1,1") -> (Pose) {
    alphaInputType=Bool
    rotationMode=BMM_Additive
    rotationSpace=BCS_BoneSpace
    translationMode=BMM_Additive
    translationSpace=BCS_BoneSpace
    alphaBoolBlend {blendInTime=1, blendOption=Linear, blendOutTime=1}
    boneToModify {boneName=dyn_cap_01_mid}
    translation {z=-2}
   }
   n63 = "Transform (Modify) Bone - Bone: dyn_cap_02_mid"!(ComponentPose=@n62, bAlphaBoolEnabled:bool=@n61, Translation="1.308,-0.000,0.327", Rotation="13.214,0.000,0.000", Scale="1,1,1") -> (Pose) {
    alphaInputType=Bool
    rotationMode=BMM_Additive
    rotationSpace=BCS_BoneSpace
    translationMode=BMM_Additive
    translationSpace=BCS_BoneSpace
    alphaBoolBlend {blendInTime=1, blendOption=Linear, blendOutTime=1}
    boneToModify {boneName=dyn_cap_02_mid}
    rotation {pitch=13.214}
    translation {x=1.308, z=0.327}
   }
   n64 = "Constraint - dyn_cap_03_mid (PPP)"!(ComponentPose=@n63, bAlphaBoolEnabled:bool=@n61, ConstraintWeights_0=0.6, ConstraintWeights_1:float=0.2, ConstraintWeights_2:float=0.2) -> (Pose) {
    alphaInputType=Bool
    alphaBoolBlend {blendInTime=1, blendOption=Linear, blendOutTime=1}
    boneToModify {boneName=dyn_cap_03_mid}
    constraintSetup [{targetBone {boneName=pelvis}}, {targetBone {boneName=thigh_l}}, {targetBone {boneName=thigh_r}}]
    constraintWeights [1,1,1]
   }
   n65 = "Transform (Modify) Bone - Bone: dyn_cap_03_mid"!(ComponentPose=@n64, bAlphaBoolEnabled:bool=@n61, Translation="0,0,2", Rotation="-10,0.000,0.000", Scale="0.8,1,1") -> (Pose) {
    alphaInputType=Bool
    rotationMode=BMM_Additive
    rotationSpace=BCS_BoneSpace
    scaleMode=BMM_Additive
    scaleSpace=BCS_BoneSpace
    translationMode=BMM_Additive
    translationSpace=BCS_BoneSpace
    alphaBoolBlend {blendInTime=1, blendOption=Linear, blendOutTime=1}
    boneToModify {boneName=dyn_cap_03_mid}
    rotation {pitch=-10}
    scale {x=0.8}
    translation {z=2}
   }
   n66 = "Transform (Modify) Bone - Bone: dyn_cap_04_mid"!(ComponentPose=@n65, bAlphaBoolEnabled:bool=@n61, Translation="0,0,0", Rotation="-10,0.000,0", Scale="1,1,1") -> (Pose) {
    alphaInputType=Bool
    rotationMode=BMM_Additive
    rotationSpace=BCS_BoneSpace
    translationMode=BMM_Additive
    translationSpace=BCS_BoneSpace
    alphaBoolBlend {blendInTime=1, blendOption=Linear, blendOutTime=1}
    boneToModify {boneName=dyn_cap_04_mid}
    rotation {pitch=-10}
   }
   n67 = "Transform (Modify) Bone - Bone: dyn_cap_a_02_r"!(ComponentPose=@n66, bAlphaBoolEnabled:bool=@n61, Translation="-1.653,0.891,-2.705", Rotation="-13.176,0.000,0.000", Scale="1,1,1") -> (Pose) {
    alphaInputType=Bool
    rotationMode=BMM_Additive
    rotationSpace=BCS_BoneSpace
    translationMode=BMM_Additive
    translationSpace=BCS_BoneSpace
    alphaBoolBlend {blendInTime=1, blendOption=Linear, blendOutTime=1}
    boneToModify {boneName=dyn_cap_a_02_r}
    rotation {pitch=-13.176}
    translation=<-1.653 0.891 -2.705>
   }
   n68 = "Constraint - dyn_cap_a_03_r (PPP)"!(ComponentPose=@n67, bAlphaBoolEnabled:bool=@n61, ConstraintWeights_0=0.8, ConstraintWeights_1:float=0, ConstraintWeights_2:float=0.2) -> (Pose) {
    alphaInputType=Bool
    alphaBoolBlend {blendInTime=1, blendOption=Linear, blendOutTime=1}
    boneToModify {boneName=dyn_cap_a_03_r}
    constraintSetup [{targetBone {boneName=pelvis}}, {targetBone {boneName=thigh_l}}, {targetBone {boneName=thigh_r}}]
    constraintWeights [1,1,1]
   }
   n69 = "Transform (Modify) Bone - Bone: dyn_cap_a_03_r"!(ComponentPose=@n68, bAlphaBoolEnabled:bool=@n61, Translation="0,0,0", Rotation="-20,0,0", Scale="0.8,1,1") -> (Pose) {
    alphaInputType=Bool
    rotationMode=BMM_Additive
    rotationSpace=BCS_BoneSpace
    scaleMode=BMM_Additive
    scaleSpace=BCS_BoneSpace
    translationMode=BMM_Additive
    translationSpace=BCS_BoneSpace
    alphaBoolBlend {blendInTime=1, blendOption=Linear, blendOutTime=1}
    boneToModify {boneName=dyn_cap_a_03_r}
    rotation {pitch=-20}
    scale {x=0.8}
   }
   n70 = "Transform (Modify) Bone - Bone: dyn_cap_a_02_l"!(ComponentPose=@n69, bAlphaBoolEnabled:bool=@n61, Translation="0.806,-0.000,1.867", Rotation="-14.369,0.000,0.000", Scale="1,1,1") -> (Pose) {
    alphaInputType=Bool
    rotationMode=BMM_Additive
    rotationSpace=BCS_BoneSpace
    translationMode=BMM_Additive
    translationSpace=BCS_BoneSpace
    alphaBoolBlend {blendInTime=1, blendOption=Linear, blendOutTime=1}
    boneToModify {boneName=dyn_cap_a_02_l}
    rotation {pitch=-14.369}
    translation {x=0.806, z=1.867}
   }
   n71 = "Constraint - dyn_cap_a_03_l (PPP)"!(ComponentPose=@n70, bAlphaBoolEnabled:bool=@n61, ConstraintWeights_0=0.8, ConstraintWeights_1:float=0.2, ConstraintWeights_2:float=0) -> (Pose) {
    alphaInputType=Bool
    alphaBoolBlend {blendInTime=1, blendOption=Linear, blendOutTime=1}
    boneToModify {boneName=dyn_cap_a_03_l}
    constraintSetup [{targetBone {boneName=pelvis}}, {targetBone {boneName=thigh_l}}, {targetBone {boneName=thigh_r}}]
    constraintWeights [1,1,1]
   }
   n72 = "Transform (Modify) Bone - Bone: dyn_cap_a_03_l"!(ComponentPose=@n71, bAlphaBoolEnabled:bool=@n61, Translation="0,0,0", Rotation="-20,0,0", Scale="0.8,1,1") -> (Pose) {
    alphaInputType=Bool
    rotationMode=BMM_Additive
    rotationSpace=BCS_BoneSpace
    scaleMode=BMM_Additive
    scaleSpace=BCS_BoneSpace
    translationMode=BMM_Additive
    translationSpace=BCS_BoneSpace
    alphaBoolBlend {blendInTime=1, blendOption=Linear, blendOutTime=1}
    boneToModify {boneName=dyn_cap_a_03_l}
    rotation {pitch=-20}
    scale {x=0.8}
   }
   Save cached pose 'trailPose'(Pose=Component To Local(ComponentPose=@n72)) {cacheName=trailPose}
   n73 = Look At - Bone: dyn_back_cape_MSTR(ComponentPose=Local To Component(LocalPose=Use cached pose 'prePose'()), Alpha=1) {
    alphaBoolBlend {blendOption=Linear}
    boneToModify {boneName=dyn_back_cape_MSTR}
    lookAtLocation=<0 -16>
    lookAtTarget {boneReference {boneName=root}}
    lookAt_Axis {axis=<1 0>}
   }
   n74 = Trail controller - Bone: dyn_cap_04_mid(ComponentPose=@n73, Alpha=0.5, FakeVelocity="0,0,0") {
    bActorSpaceFakeVel=true
    bShowTrailLocation=true
    bUsePlanarLimit=true
    chainLength=4
    debugLifeTime=0.1
    maxDeltaTime=0.0333
    alphaScaleBiasClamp {bInterpResult=true}
    baseJoint {boneName=root}
    planarLimits [{drivingBone {boneName=pelvis}, planeTransform {location=<-50 -10 0>, rotation=<rot 0 0 -90>, scale=<1 1 1>}}]
    rotationLimits [{limitMax=<180 180 180>}, {limitMax=<180 180 180>}, {limitMax=<180 180 180>}, {limitMax=<180 180 180>}]
    rotationOffsets [<0 0 0>, <0 0 0>, <0 0 0>, <0 0 0>]
    trailBone {boneName=dyn_cap_04_mid}
    trailRelaxationSpeed {editorCurveData {keys [<at 0 30 RCIM_Linear>,<at 1 10 RCIM_Linear>]}}
   }
   n75 = Trail controller - Bone: dyn_cap_a_03_l(ComponentPose=@n74, Alpha=0.5, FakeVelocity="0,0,0") {
    bActorSpaceFakeVel=true
    bShowTrailLocation=true
    bUsePlanarLimit=true
    chainLength=3
    debugLifeTime=0.1
    maxDeltaTime=0.0333
    alphaScaleBiasClamp {bInterpResult=true}
    baseJoint {boneName=root}
    planarLimits [{drivingBone {boneName=pelvis}, planeTransform {location=<-50 -10 0>, rotation=<rot 0 0 -90>, scale=<1 1 1>}}]
    rotationLimits [{limitMax=<180 180 180>}, {limitMax=<180 180 180>}, {limitMax=<180 180 180>}]
    rotationOffsets [<0 0 0>, <0 0 0>, <0 0 0>]
    trailBone {boneName=dyn_cap_a_03_l}
    trailRelaxationSpeed {editorCurveData {keys [<at 0 30 RCIM_Linear>,<at 1 10 RCIM_Linear>]}}
   }
   n76 = Trail controller - Bone: dyn_cap_a_03_r(ComponentPose=@n75, Alpha=0.5, FakeVelocity="0,0,0") {
    bActorSpaceFakeVel=true
    bShowTrailLocation=true
    bUsePlanarLimit=true
    chainLength=3
    debugLifeTime=0.1
    maxDeltaTime=0.0333
    alphaScaleBiasClamp {bInterpResult=true}
    baseJoint {boneName=root}
    planarLimits [{drivingBone {boneName=pelvis}, planeTransform {location=<-50 -10 0>, rotation=<rot 0 0 -90>, scale=<1 1 1>}}]
    rotationLimits [{limitMax=<180 180 180>}, {limitMax=<180 180 180>}, {limitMax=<180 180 180>}]
    rotationOffsets [<0 0 0>, <0 0 0>, <0 0 0>]
    trailBone {boneName=dyn_cap_a_03_r}
    trailRelaxationSpeed {editorCurveData {keys [<at 0 30 RCIM_Linear>,<at 1 10 RCIM_Linear>]}}
   }
   Save cached pose 'frontendPose'(Pose=Component To Local(ComponentPose=@n76)) {cacheName=frontendPose}
   SelectVector(A="0,0,-980", B="0, 0, 0", bPickA=Get bIsDBNO())
   Get bIsTacticalSprinting()
   Get capeCtrlRigAlpha() -> (capeCtrlRigAlpha:double)
  ]}
 EventGraph {nodes [
   Event BlueprintUpdateAnimation()
   > n1 = Sequence!(execute) -> (then_0, then_1, then_2:Exec, then_3:Exec, then_4:Exec, then_5:Exec, then_6:Exec, then_7:Exec)
   > n2 = Gravity Override State Function DEPRECATED(CharacterPartAnimInstance=Self-Reference(), GameTimeInSeconds=GetGameTimeInSeconds(), GravityOverrideParams=Get Gravity Override Params())
   > Set GravityOverride(GravityOverride=@n2)
   n3 = ComplexSimSpaceFunction DEPRECATED!(
    execute=@n1.then_1
    CharacterPartAnimInstance:Custom Character Part Anim Instance Object Reference=Self-Reference()
    InterpSpeedMax:double=9
    PawnSpeedMax:double=20
    RInterpErrorTolerance:double=1
    DeltaTime:double=0
    DebugDraw:bool=false
    sim_space_rotation_in:Rotator=Get sim_space_rotation() -> (sim_space_rotation:Rotator)
    inComponentSpaceIn:bool=Get InComponentSpace() -> (InComponentSpace:bool)
    inComponentSpaceLastFrameIn:bool=Get InComponentSpaceLastFrame() -> (InComponentSpaceLastFrame:bool)
    lockedOnSimSpaceIn:bool=Get LockedOnSimSpace() -> (LockedOnSimSpace:bool)
    InterpSpeedIn:double=Get InterpSpeed() -> (InterpSpeed:double)
   ) -> (
    then
    sim_space_rotation:Rotator
    inComponentSpaceOut:bool
    inComponentSpaceLastFrameOut:bool
    lockedOnSimSpaceOut:bool
    InterpSpeedOut:double
   )
   > Set sim_space_rotation(sim_space_rotation:Rotator=@n3) -> (Output_Get:Rotator)
   > Set InComponentSpace(InComponentSpace:bool=@n3.inComponentSpaceOut) -> (Output_Get:bool)
   > Set InComponentSpaceLastFrame(InComponentSpaceLastFrame:bool=@n3.inComponentSpaceLastFrameOut) -> (Output_Get:bool)
   > Set LockedOnSimSpace(LockedOnSimSpace:bool=@n3.lockedOnSimSpaceOut) -> (Output_Get:bool)
   > Set InterpSpeed(InterpSpeed:double=@n3.InterpSpeedOut) -> (Output_Get:double)
   n4 = BreakTransform(InTransform=GetDeltaTransformFromRefPose(self=GetOwnerSkeletalMeshComponent(), BoneName=thigh_r, BaseName=pelvis))
   Set thigh_rot_up_r(execute=@n1.then_3, thigh_rot_up_r:double=MapRangeClamped(Value=BreakRotator(InRot=@n4.Rotation).Yaw, InRangeA=-15, InRangeB=-100)) -> (Output_Get:double)
   n5 = BreakTransform(InTransform=GetDeltaTransformFromRefPose(self=GetOwnerSkeletalMeshComponent(), BoneName=thigh_l, BaseName=pelvis))
   > Set thigh_rot_up_l(thigh_rot_up_l:double=MapRangeClamped(Value=BreakRotator(InRot=@n5.Rotation).Yaw, InRangeA=-15, InRangeB=-100)) -> (Output_Get:double)
   n6 = BreakTransform(InTransform=GetDeltaTransformFromRefPose(self=GetOwnerSkeletalMeshComponent(), BoneName=upperarm_r, BaseName=spine_05))
   > Set arm_rot_bk_r(arm_rot_bk_r:double=MapRangeClamped(Value=BreakRotator(InRot=@n6.Rotation).Yaw, InRangeA=0, InRangeB=-20)) -> (Output_Get:double)
   > Set arm_rot_bk_l(arm_rot_bk_l:double=0) -> (Output_Get:double)
   n7 = Get SkelMesh()
   > Set leftCapeConstraintAlpha(leftCapeConstraintAlpha:double=GetAlphaFromReferenceTransformDelta(
     SkeletalMeshComponent=@n7
     BoneName=upperarm_l
     BaseName=clavicle_l
     DeltaComponentType=QuaternionTwist
     InputRange=(X=0,Y=25)
     TwistAxis="1,0,0"
    )) -> (Output_Get:double)
   > Set rightCapeConstraintAlpha(rightCapeConstraintAlpha:double=GetAlphaFromReferenceTransformDelta(
     SkeletalMeshComponent=@n7
     BoneName=upperarm_r
     BaseName=clavicle_r
     DeltaComponentType=QuaternionTwist
     InputRange=(X=0,Y=25)
     TwistAxis="1,0,0"
    )) -> (Output_Get:double)
   n8 = NOT Boolean(A=Get bIsSkydiveFloating())
   n9 = Get bIsSkydiveDiveMode()
   n10 = NOT Boolean(A=@n9)
   Set capeCtrlRigAlpha(execute=@n1.then_4, capeCtrlRigAlpha:double="To Float (Boolean)"(InBool=AND Boolean(A=@n10, B=@n8))) -> (Output_Get:double)
   > Set isDiving(isDiving:double="To Float (Boolean)"(InBool=@n9)) -> (Output_Get:double)
   n11 = PlayerMovement(execute=@n1.then_5, AnimInstance=Self-Reference(), XDotProductCoefficient=1, YDotProductCoefficient=1, ZDotProductCoefficient=1)
   n12 = MapRangeClamped(Value=@n11.YMovement, InRangeA=0, InRangeB=-1)
   n13 = float * float(A="Absolute (Float)"(A=@n11), B=0.5)
   > Set BwdMovement(BwdMovement:double=MapRangeClamped(
     Value=float + float(A=@n13, B=@n12) {is=title K2Node_CommutativeAssociativeBinaryOperator}
     InRangeA=0
     InRangeB=1.1
     OutRangeA=0.5
     OutRangeB=0
    )) -> (Output_Get:double)
   n14 = Get bIsDBNO()
   n15 = Branch(execute=@n1.then_6, Condition=@n14)
   > Set BwdMovement(BwdMovement=float * float(A=Get BwdMovement(), B=0.3))
   > Set DBNO_alpha(DBNO_alpha:bool=true) -> (Output_Get:bool)
   Set DBNO_alpha(execute=@n15.else, DBNO_alpha=false)
   > n16 = Branch(Condition=Get bIsSkyDiving())
   > n17 = Branch(Condition="float > float"(A=Get PawnVelocityZ(), B=50) {is=title K2Node_CallFunction})
   > Set BwdMovement(BwdMovement=float * float(A=Get BwdMovement(), B=0))
   Set BwdMovement(execute=[@n17.else, @n16.else], BwdMovement=float * float(A=Get BwdMovement(), B=1))
   n18 = GetPartSkeletalMeshComponent()
   Is Valid(exec=@n1.then_7, InputObject=@n18)
   > Delay(Duration=0)
   > SetVectorParameterValueOnMaterials!(
    execute
    self:Mesh Component Object Reference=@n18
    ParameterName:Name=Cube Location
    ParameterValue:Vector=TransformLocation!(T:"Transform (by ref)"=GetSocketTransform(self=@n18, InSocketName=spine_05, TransformSpace=RTS_Component), Location:Vector="0, 0, 0") -> (ReturnValue:Vector)
   ) -> (then)
   > SetVectorParameterValueOnMaterials(
    self=@n18
    ParameterName=Core Location R
    ParameterValue=TransformLocation(T=GetSocketTransform(self=@n18, InSocketName=thigh_twist_01_r, TransformSpace=RTS_Component), Location="0, 0, 0")
   )
   > SetVectorParameterValueOnMaterials(
    self=@n18
    ParameterName=Core Location L
    ParameterValue=TransformLocation(T=GetSocketTransform(self=@n18, InSocketName=thigh_twist_01_l, TransformSpace=RTS_Component), Location="0, 0, 0")
   )
   n19 = BreakTransform(InTransform=GetDeltaTransformFromRefPose(self=GetOwnerSkeletalMeshComponent(), BoneName=calf_l, BaseName=thigh_l))
   Set knee_rot_l(execute=@n1.then_2, knee_rot_l:double=MapRangeClamped(Value=BreakRotator(InRot=@n19.Rotation).Yaw, InRangeA=10, InRangeB=110)) -> (Output_Get:double)
   n20 = BreakTransform(InTransform=GetDeltaTransformFromRefPose(self=GetOwnerSkeletalMeshComponent(), BoneName=head, BaseName=spine_05))
   > Set head_rot_bk(head_rot_bk:double=MapRangeClamped(Value=BreakRotator(InRot=@n20.Rotation).Yaw, InRangeA=0, InRangeB=-50)) -> (Output_Get:double)
   Event BlueprintInitializeAnimation()
   n21 = GetOwnerSkeletalMeshComponent()
   > Is Valid(InputObject=@n21)
   > Set SkelMesh(SkelMesh=@n21)
   OR Boolean(A=@n14, B=Get bIsDBNOCarried())
  ]}
]
variables [
 arm_rot_bk_l {type=number}
 arm_rot_bk_r {type=number}
 BwdMovement {type=number}
 capeCtrlRigAlpha {type=number}
 DBNO_alpha {type=boolean}
 Gravity Override Params {
  type=GravityOverrideParamsStruct
  default {
   skydive_FloatingDown {gravityOverride=<0 -700 100>, windAmplitude=<200 -500 200>, windFrequency=<1.5 5 6>, windOffset=<0 -600 100>}
   skydive_Parachuting {gravityOverride=<0 -600 -980>, windAmplitude=<200 -200 300>, windFrequency=<1.5 5 4>}
  }
 }
 GravityOverride {type=Vector}
 head_rot_bk {type=number}
 InComponentSpace {type=boolean}
 InComponentSpaceLastFrame {type=boolean}
 InterpSpeed {type=number}
 isDiving {type=number}
 knee_rot_l {type=number}
 leftCapeConstraintAlpha {type=number}
 LockedOnSimSpace {type=boolean}
 rightCapeConstraintAlpha {type=number}
 sim_space_rotation {type=Rotator}
 SkelMesh {type=SkeletalMeshComponent}
 thigh_rot_up_l {type=number}
 thigh_rot_up_r {type=number}
]
