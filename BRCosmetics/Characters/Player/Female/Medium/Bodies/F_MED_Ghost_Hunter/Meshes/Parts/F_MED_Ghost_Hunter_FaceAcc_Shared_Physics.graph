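# Physics asset for the Ghost Hunter female medium face-accessory mesh, defining two dynamic hair tails and upper-body collision shapes.
# Each tail is a capsule chain from a head-linked base, with limited swing and very tight twist between segments.
# Capsules on head, neck, upper spine, clavicles, upper arms and hands form the body volume.
# Body constraints leave rotation free, except the limited left upper-arm joint; the two hand capsules also differ.
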PhysicsAsset /BRCosmetics/Characters/Player/Female/Medium/Bodies/F_MED_Ghost_Hunter/Meshes/Parts/F_MED_Ghost_Hunter_FaceAcc_Shared_Physics
bodies {
 clavicle_l [{shapeType=Capsule, length=8.6164, radius=4.166, center=<8.2306 -1.2217 0.6826>, rotation=<rot 71.3243 2.1586 5.1544>}]
 clavicle_r [{shapeType=Capsule, length=8.6164, radius=4.166, center=<-8.2306 1.2217 -0.6826>, rotation=<rot 71.3271 2.1586 -174.8459>}]
 dyn_hair_tail_a_base_l [{shapeType=Capsule, length=0.505, radius=0.505, rotation=<rot 0 -90 0>}]
 dyn_hair_tail_a_base_r [{shapeType=Capsule, length=0.505, radius=0.505, rotation=<rot 0 -90 0>}]
 dyn_hair_tail_a_l [{shapeType=Capsule, length=0.5679, radius=2.1931, center=<4 0 0>, rotation=<rot 90 0 0>}]
 dyn_hair_tail_a_r [{shapeType=Capsule, length=0.5679, radius=2.1931, center=<4 0 0>, rotation=<rot 90 0 0>}]
 dyn_hair_tail_b_l [{shapeType=Capsule, length=1.3048, radius=1.976, center=<2.5 0 0>, rotation=<rot 90 0 0>}]
 dyn_hair_tail_b_r [{shapeType=Capsule, length=1.1907, radius=1.976, center=<2.5 0 0>, rotation=<rot 90 0 0>}]
 dyn_hair_tail_c_l [{shapeType=Capsule, length=1.4717, radius=1.7836, center=<3 0 0>, rotation=<rot 90 0 0>}]
 dyn_hair_tail_c_r [{shapeType=Capsule, length=1.3576, radius=1.7836, center=<3 0 0>, rotation=<rot 90 0 0>}]
 dyn_hair_tail_d_l [{shapeType=Capsule, length=2.0862, radius=1.7526, center=<3 0 0>, rotation=<rot 90 0 0>}]
 dyn_hair_tail_d_r [{shapeType=Capsule, length=1.9721, radius=1.7526, center=<3 0 0>, rotation=<rot 90 0 0>}]
 hand_l [{shapeType=Capsule, length=6.5, radius=3.25, center=<10.3763 -1.2472 0.1706>, rotation=<rot 8.9346 0 0>}]
 hand_r [{shapeType=Capsule, length=8.5, radius=7, center=<-3.8275 5.3439 1.086>, rotation=<rot 90 -20.556 -21.7686>}]
 head [{shapeType=Capsule, length=8.77, radius=5.48, center=<3.1816 1.3754 0.0001>, rotation=<rot 90 0 0>}]
 neck_01 [{shapeType=Capsule, length=8.1701, radius=3.0799, center=<0.8316 -0.1459 -0.1493>, rotation=<rot 90 -10.6197 -13.3382>}]
 spine_04 [{shapeType=Capsule, length=9.7577, radius=11, center=<3.3548 0.6133 0.0001>}]
 spine_05 [
  {shapeType=Capsule, length=15.3442, radius=10.3707, center=<1.6834 -2.0186 0>}
  {shapeType=Capsule, length=11.5187, radius=7.1598, shapeName=spine_05_capsule, center=<7.6112 -3.4154 0>}
 ]
 upperarm_l [{shapeType=Capsule, length=18, radius=3.5, center=<11.0088 -0.6173 0.4163>, rotation=<rot 89.5306 180 -179.0423>}]
 upperarm_r [{shapeType=Capsule, length=18, radius=3.5, center=<-11.0088 0.6173 -0.4163>, rotation=<rot 89.5314 180 0.9577>}]
}
constraints [
 {
  bone1Name=head
  bone2Name=neck_01
  swing1LimitDegrees=45
  swing1Motion=Free
  swing2LimitDegrees=45
  swing2Motion=Free
  twistLimitDegrees=45
  twistMotion=Free
 }
 {
  bone1Name=clavicle_l
  bone2Name=spine_05
  swing1LimitDegrees=45
  swing1Motion=Free
  swing2LimitDegrees=45
  swing2Motion=Free
  twistLimitDegrees=45
  twistMotion=Free
 }
 {
  bone1Name=spine_05
  bone2Name=spine_04
  swing1LimitDegrees=45
  swing1Motion=Free
  swing2LimitDegrees=45
  swing2Motion=Free
  twistLimitDegrees=45
  twistMotion=Free
 }
 {
  bone1Name=neck_01
  bone2Name=spine_05
  swing1LimitDegrees=45
  swing1Motion=Free
  swing2LimitDegrees=45
  swing2Motion=Free
  twistLimitDegrees=45
  twistMotion=Free
 }
 {
  bone1Name=dyn_hair_tail_a_base_l
  bone2Name=head
  swing1LimitDegrees=45
  swing1Motion=Free
  swing2LimitDegrees=45
  swing2Motion=Free
  twistLimitDegrees=45
  twistMotion=Free
 }
 {
  bone1Name=dyn_hair_tail_a_base_r
  bone2Name=head
  swing1LimitDegrees=45
  swing1Motion=Free
  swing2LimitDegrees=45
  swing2Motion=Free
  twistLimitDegrees=45
  twistMotion=Free
 }
 {bone1Name=dyn_hair_tail_a_l, bone2Name=dyn_hair_tail_a_base_l, swing1LimitDegrees=35, swing2LimitDegrees=35, twistLimitDegrees=2}
 {bone1Name=dyn_hair_tail_b_l, bone2Name=dyn_hair_tail_a_l, swing1LimitDegrees=35, swing2LimitDegrees=35, twistLimitDegrees=2}
 {bone1Name=dyn_hair_tail_c_l, bone2Name=dyn_hair_tail_b_l, swing1LimitDegrees=35, swing2LimitDegrees=35, twistLimitDegrees=2}
 {bone1Name=dyn_hair_tail_d_l, bone2Name=dyn_hair_tail_c_l, swing1LimitDegrees=35, swing2LimitDegrees=35, twistLimitDegrees=2}
 {bone1Name=dyn_hair_tail_d_r, bone2Name=dyn_hair_tail_c_r, swing1LimitDegrees=35, swing2LimitDegrees=35, twistLimitDegrees=2}
 {
  bone1Name=upperarm_r
  bone2Name=clavicle_r
  swing1LimitDegrees=45
  swing1Motion=Free
  swing2LimitDegrees=45
  swing2Motion=Free
  twistLimitDegrees=45
  twistMotion=Free
 }
 {bone1Name=dyn_hair_tail_a_r, bone2Name=dyn_hair_tail_a_base_r, swing1LimitDegrees=35, swing2LimitDegrees=35, twistLimitDegrees=2}
 {bone1Name=dyn_hair_tail_b_r, bone2Name=dyn_hair_tail_a_r, swing1LimitDegrees=35, swing2LimitDegrees=35, twistLimitDegrees=2}
 {bone1Name=dyn_hair_tail_c_r, bone2Name=dyn_hair_tail_b_r, swing1LimitDegrees=35, swing2LimitDegrees=35, twistLimitDegrees=2}
 {bone1Name=upperarm_l, bone2Name=clavicle_l, swing1LimitDegrees=45, swing2LimitDegrees=15, twistLimitDegrees=5}
 {
  bone1Name=clavicle_r
  bone2Name=spine_05
  swing1LimitDegrees=45
  swing1Motion=Free
  swing2LimitDegrees=45
  swing2Motion=Free
  twistLimitDegrees=45
  twistMotion=Free
 }
 {
  bone1Name=hand_l
  bone2Name=upperarm_l
  swing1LimitDegrees=45
  swing1Motion=Free
  swing2LimitDegrees=45
  swing2Motion=Free
  twistLimitDegrees=45
  twistMotion=Free
 }
 {
  bone1Name=hand_r
  bone2Name=upperarm_r
  swing1LimitDegrees=45
  swing1Motion=Free
  swing2LimitDegrees=45
  swing2Motion=Free
  twistLimitDegrees=45
  twistMotion=Free
 }
]
