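# Post-processes the Ghost Hunter face accessory skeleton to animate its two hair tails.
# bRigidBodyEnabled picks between head-space RigidBody simulation, fed by parent pose with body-mesh arms layered in, and a constraint/Look At fallback following spine, head and clavicle.
# Both paths add bone offsets weighted by frontendPoseAlpa, active in the front end outside emotes.
# Customization load and emote start or end trigger a delayed dynamics reset.
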AnimBlueprint /BRCosmetics/Characters/Player/Female/Medium/Bodies/F_MED_Ghost_Hunter/Meshes/Parts/F_MED_Ghost_Hunter_FaceAcc_AnimBP
target_skeleton=&/F_MED_Ghost_Hunter_FaceAcc_Skeleton
graphs [
 AnimGraph {nodes [
   Output Pose(Result=Blend Poses by bool(BlendPose_0=Use cached pose 'rigidPose'(), BlendPose_1=Use cached pose 'constrainPose'(), bActiveValue=Get bRigidBodyEnabled()) {blendType=Linear}) {layerGroup=None}
   Save cached pose 'prePose'(Pose=Copy Pose From Mesh() {bUseAttachedParent=true}) {cacheName=prePose}
   n1 = Copy Pose From Mesh(SourceMeshComponent=Get Body_Mesh())
   n2 = Use cached pose 'prePose'()
   n3 = Get frontendPoseAlpa() -> (frontendPoseAlpa:double)
   n4 = "Transform (Modify) Bone - Bone: dyn_hair_tail_a_base_r"(
    ComponentPose=Local To Component(LocalPose=Layered blend per bone(BasePose=@n2, BlendPoses_0=@n1) {layerSetup [{branchFilters [{boneName=upperarm_l}, {boneName=upperarm_r}]}]})
    Alpha=@n3
    Translation="0,-3.5,0"
    Rotation="10,-40,0"
    Scale="1,1,1"
   ) {
    rotationMode=BMM_Additive
    rotationSpace=BCS_BoneSpace
    translationMode=BMM_Additive
    translationSpace=BCS_BoneSpace
    alphaBoolBlend {blendOption=Linear}
    alphaScaleBiasClamp {interpSpeedDecreasing=15, interpSpeedIncreasing=15}
    boneToModify {boneName=dyn_hair_tail_a_base_r}
    rotation {pitch=10, yaw=-40}
    translation {y=-3.5}
   }
   n5 = "Transform (Modify) Bone - Bone: dyn_hair_tail_a_base_l"(ComponentPose=@n4, Alpha=@n3, Translation="0.205,2,0", Rotation="30,-40,0", Scale="1,1,1") {
    rotationMode=BMM_Additive
    rotationSpace=BCS_BoneSpace
    translationMode=BMM_Additive
    translationSpace=BCS_BoneSpace
    alphaBoolBlend {blendOption=Linear}
    alphaScaleBiasClamp {interpSpeedDecreasing=15, interpSpeedIncreasing=15}
    boneToModify {boneName=dyn_hair_tail_a_base_l}
    rotation {pitch=30, yaw=-40}
    translation=<0.205 2>
   }
   n6 = RigidBody!(ComponentPose=@n5, Alpha, ExternalForce, bTransferBoneVelocities:bool=False) -> (Pose) {
    bClampLinearTranslationLimitToRefPose=true
    bOverrideWorldGravity=true
    bTransferBoneVelocities=true
    overridePhysicsAsset=&/F_MED_Ghost_Hunter_FaceAcc_Shared_Physics
    simulationSpace=BaseBoneSpace
    alphaBoolBlend {blendOption=Linear}
    baseBoneRef {boneName=head}
    componentAppliedLinearAccClamp=<3000 3000 3000>
    componentLinearAccScale=<0.1 0.1 0.1>
    componentLinearVelScale=<0.1 0.1 0.1>
    overrideWorldGravity {z=-980}
   }
   n7 = "Transform (Modify) Bone - Bone: dyn_hair_tail_b_l"(ComponentPose=@n6, Alpha=@n3, Translation="0,2,0", Rotation="-5,-2,0", Scale="1,1,0.8") {
    rotationMode=BMM_Additive
    rotationSpace=BCS_BoneSpace
    scaleMode=BMM_Additive
    translationMode=BMM_Additive
    translationSpace=BCS_BoneSpace
    alphaBoolBlend {blendOption=Linear}
    alphaScaleBiasClamp {bInterpResult=true, interpSpeedDecreasing=15, interpSpeedIncreasing=15}
    boneToModify {boneName=dyn_hair_tail_b_l}
    rotation {pitch=-5, yaw=-2}
    scale {z=0.8}
    translation {y=2}
   }
   n8 = "Transform (Modify) Bone - Bone: dyn_hair_tail_c_l"(ComponentPose=@n7, Alpha=@n3, Translation="0,-1,0", Rotation="0,0,0", Scale="1,1,0.8") {
    rotationMode=BMM_Additive
    rotationSpace=BCS_BoneSpace
    scaleMode=BMM_Additive
    translationMode=BMM_Additive
    translationSpace=BCS_BoneSpace
    alphaBoolBlend {blendOption=Linear}
    alphaScaleBiasClamp {bInterpResult=true, interpSpeedDecreasing=15, interpSpeedIncreasing=15}
    boneToModify {boneName=dyn_hair_tail_c_l}
    scale {z=0.8}
    translation {y=-1}
   }
   n9 = "Transform (Modify) Bone - Bone: dyn_hair_tail_c_r"(ComponentPose=@n8, Alpha=@n3, Translation="-1,-1.5,0", Rotation="0,0,0", Scale="1,1,0.8") {
    rotationMode=BMM_Additive
    rotationSpace=BCS_BoneSpace
    scaleMode=BMM_Additive
    translationMode=BMM_Additive
    translationSpace=BCS_BoneSpace
    alphaBoolBlend {blendOption=Linear}
    alphaScaleBiasClamp {bInterpResult=true, interpSpeedDecreasing=15, interpSpeedIncreasing=15}
    boneToModify {boneName=dyn_hair_tail_c_r}
    scale {z=0.8}
    translation=<-1 -1.5>
   }
   Save cached pose 'rigidPose'(Pose=Component To Local(ComponentPose=@n9)) {cacheName=rigidPose}
   n10 = Get frontendPoseAlpa()
   n11 = "Constraint - dyn_hair_tail_a_base_l (PP)"!(ComponentPose=Local To Component(LocalPose=Use cached pose 'prePose'()), Alpha=1, ConstraintWeights_0=0.4, ConstraintWeights_1:float=0.6) -> (Pose) {
    alphaBoolBlend {blendOption=Linear}
    boneToModify {boneName=dyn_hair_tail_a_base_l}
    constraintSetup [{targetBone {boneName=spine_05}}, {targetBone {boneName=head}}]
    constraintWeights [1,1]
   }
   n12 = Look At - Bone: dyn_hair_tail_a_l(ComponentPose=@n11, Alpha=1) {
    alphaBoolBlend {blendOption=Linear}
    boneToModify {boneName=dyn_hair_tail_a_l}
    lookAtLocation {x=0}
    lookAtTarget {boneReference {boneName=dyn_hair_tail_b_l}}
    lookAt_Axis {axis=<1 0>}
   }
   n13 = "Constraint - dyn_hair_tail_b_l (PP)"!(ComponentPose=@n12, Alpha=1, ConstraintWeights_0=0.7, ConstraintWeights_1:float=0.3) -> (Pose) {
    alphaBoolBlend {blendOption=Linear}
    boneToModify {boneName=dyn_hair_tail_b_l}
    constraintSetup [{targetBone {boneName=spine_05}}, {targetBone {boneName=clavicle_l}}]
    constraintWeights [1,1]
   }
   n14 = "Constraint - dyn_hair_tail_c_l (PPP)"!(ComponentPose=@n13, Alpha=1, ConstraintWeights_0=0.9, ConstraintWeights_1:float=0.1, ConstraintWeights_2:float=0) -> (Pose) {
    alphaBoolBlend {blendOption=Linear}
    boneToModify {boneName=dyn_hair_tail_c_l}
    constraintSetup [{targetBone {boneName=spine_05}}, {targetBone {boneName=clavicle_l}}, {targetBone {boneName=head}}]
    constraintWeights [1,1,1]
   }
   n15 = "Constraint - dyn_hair_tail_d_l (P)"(ComponentPose=@n14, Alpha=1, ConstraintWeights_0=1) {
    alphaBoolBlend {blendOption=Linear}
    boneToModify {boneName=dyn_hair_tail_d_l}
    constraintSetup [{targetBone {boneName=spine_05}}]
    constraintWeights [1]
   }
   n16 = "Constraint - dyn_hair_tail_a_base_r (PP)"!(ComponentPose=@n15, Alpha=1, ConstraintWeights_0=0.4, ConstraintWeights_1:float=0.6) -> (Pose) {
    alphaBoolBlend {blendOption=Linear}
    boneToModify {boneName=dyn_hair_tail_a_base_r}
    constraintSetup [{targetBone {boneName=spine_05}}, {targetBone {boneName=head}}]
    constraintWeights [1,1]
   }
   n17 = Look At - Bone: dyn_hair_tail_a_r(ComponentPose=@n16, Alpha=1) {
    alphaBoolBlend {blendOption=Linear}
    boneToModify {boneName=dyn_hair_tail_a_r}
    lookAtLocation {x=0}
    lookAtTarget {boneReference {boneName=dyn_hair_tail_b_r}}
    lookAt_Axis {axis=<1 0>}
   }
   n18 = "Constraint - dyn_hair_tail_b_r (PP)"!(ComponentPose=@n17, Alpha=1, ConstraintWeights_0=0.7, ConstraintWeights_1:float=0.3) -> (Pose) {
    alphaBoolBlend {blendOption=Linear}
    boneToModify {boneName=dyn_hair_tail_b_r}
    constraintSetup [{targetBone {boneName=spine_05}}, {targetBone {boneName=clavicle_l}}]
    constraintWeights [1,1]
   }
   n19 = "Constraint - dyn_hair_tail_c_r (PPP)"!(ComponentPose=@n18, Alpha=1, ConstraintWeights_0=0.9, ConstraintWeights_1:float=0.1, ConstraintWeights_2:float=0) -> (Pose) {
    alphaBoolBlend {blendOption=Linear}
    boneToModify {boneName=dyn_hair_tail_c_r}
    constraintSetup [{targetBone {boneName=spine_05}}, {targetBone {boneName=clavicle_l}}, {targetBone {boneName=head}}]
    constraintWeights [1,1,1]
   }
   n20 = "Constraint - dyn_hair_tail_d_r (P)"(ComponentPose=@n19, Alpha=1, ConstraintWeights_0=1) {
    alphaBoolBlend {blendOption=Linear}
    boneToModify {boneName=dyn_hair_tail_d_r}
    constraintSetup [{targetBone {boneName=spine_05}}]
    constraintWeights [1]
   }
   n21 = "Transform (Modify) Bone - Bone: dyn_hair_tail_a_r"(ComponentPose=@n20, Alpha=@n10, Translation="0.5,-3,0", Rotation="0,-2,0", Scale="1,1,1") {
    rotationMode=BMM_Additive
    rotationSpace=BCS_BoneSpace
    translationMode=BMM_Additive
    translationSpace=BCS_BoneSpace
    alphaBoolBlend {blendOption=Linear}
    alphaScaleBiasClamp {bInterpResult=true, interpSpeedDecreasing=15, interpSpeedIncreasing=15}
    boneToModify {boneName=dyn_hair_tail_a_r}
    rotation {yaw=-2}
    translation=<0.5 -3>
   }
   n22 = "Transform (Modify) Bone - Bone: dyn_hair_tail_b_r"(ComponentPose=@n21, Alpha=@n10, Translation="0.009,-1.5,0.854", Rotation="7.008,5,0.579", Scale="1,1,1") {
    rotationMode=BMM_Additive
    rotationSpace=BCS_BoneSpace
    translationMode=BMM_Additive
    translationSpace=BCS_BoneSpace
    alphaBoolBlend {blendOption=Linear}
    alphaScaleBiasClamp {bInterpResult=true, interpSpeedDecreasing=15, interpSpeedIncreasing=15}
    boneToModify {boneName=dyn_hair_tail_b_r}
    rotation=<rot 7.008 5 0.579>
    translation=<0.009 -1.5 0.854>
   }
   n23 = "Transform (Modify) Bone - Bone: dyn_hair_tail_c_r"(ComponentPose=@n22, Alpha=@n10, Translation="0,0,0", Rotation="6.487,17.604,0.360", Scale="1,1,1") {
    rotationMode=BMM_Additive
    rotationSpace=BCS_BoneSpace
    translationMode=BMM_Additive
    translationSpace=BCS_BoneSpace
    alphaBoolBlend {blendOption=Linear}
    alphaScaleBiasClamp {bInterpResult=true, interpSpeedDecreasing=15, interpSpeedIncreasing=15}
    boneToModify {boneName=dyn_hair_tail_c_r}
    rotation=<rot 6.487 17.6041 0.36>
   }
   n24 = "Transform (Modify) Bone - Bone: dyn_hair_tail_d_r"(ComponentPose=@n23, Alpha=@n10, Translation="0,0,0", Rotation="6.298,14.313,1.603", Scale="1,1,1") {
    rotationMode=BMM_Additive
    rotationSpace=BCS_BoneSpace
    translationMode=BMM_Additive
    translationSpace=BCS_BoneSpace
    alphaBoolBlend {blendOption=Linear}
    alphaScaleBiasClamp {bInterpResult=true, interpSpeedDecreasing=15, interpSpeedIncreasing=15}
    boneToModify {boneName=dyn_hair_tail_d_r}
    rotation=<rot 6.298 14.3133 1.603>
   }
   n25 = "Transform (Modify) Bone - Bone: dyn_hair_tail_a_l"(ComponentPose=@n24, Alpha=@n10, Translation="0,-2.564,-0.248", Rotation="8.5,6.889,0.488", Scale="1,1,1") {
    rotationMode=BMM_Additive
    rotationSpace=BCS_BoneSpace
    translationMode=BMM_Additive
    translationSpace=BCS_BoneSpace
    alphaBoolBlend {blendOption=Linear}
    alphaScaleBiasClamp {bInterpResult=true, interpSpeedDecreasing=15, interpSpeedIncreasing=15}
    boneToModify {boneName=dyn_hair_tail_a_l}
    rotation=<rot 8.5 6.889 0.488>
    translation {y=-2.564, z=-0.248}
   }
   n26 = "Transform (Modify) Bone - Bone: dyn_hair_tail_b_l"(ComponentPose=@n25, Alpha=@n10, Translation="-0.072,-0.003,0.5", Rotation="19.438,6.218,11.475", Scale="1,1,1") {
    rotationMode=BMM_Additive
    rotationSpace=BCS_BoneSpace
    translationMode=BMM_Additive
    translationSpace=BCS_BoneSpace
    alphaBoolBlend {blendOption=Linear}
    alphaScaleBiasClamp {bInterpResult=true, interpSpeedDecreasing=15, interpSpeedIncreasing=15}
    boneToModify {boneName=dyn_hair_tail_b_l}
    rotation=<rot 19.438 6.218 11.4752>
    translation=<-0.072 -0.003 0.5>
   }
   n27 = "Transform (Modify) Bone - Bone: dyn_hair_tail_c_l"(ComponentPose=@n26, Alpha=@n10, Translation="0.010,0.000,0.386", Rotation="-15.295,0.000,0.000", Scale="1,1,1") {
    rotationMode=BMM_Additive
    rotationSpace=BCS_BoneSpace
    translationMode=BMM_Additive
    translationSpace=BCS_BoneSpace
    alphaBoolBlend {blendOption=Linear}
    alphaScaleBiasClamp {bInterpResult=true, interpSpeedDecreasing=15, interpSpeedIncreasing=15}
    boneToModify {boneName=dyn_hair_tail_c_l}
    rotation {pitch=-15.295}
    translation {x=0.01, z=0.386}
   }
   n28 = "Transform (Modify) Bone - Bone: dyn_hair_tail_d_l"(ComponentPose=@n27, Alpha=@n10, Translation="0,0,0", Rotation="0,0,0", Scale="1,1,1") {
    rotationMode=BMM_Additive
    rotationSpace=BCS_BoneSpace
    translationMode=BMM_Additive
    translationSpace=BCS_BoneSpace
    alphaBoolBlend {blendOption=Linear}
    alphaScaleBiasClamp {bInterpResult=true, interpSpeedDecreasing=15, interpSpeedIncreasing=15}
    boneToModify {boneName=dyn_hair_tail_d_l}
   }
   Save cached pose 'constrainPose'(Pose=Component To Local(ComponentPose=@n28)) {cacheName=constrainPose}
  ]}
 EventGraph {nodes [
   Event BlueprintUpdateAnimation()
   n1 = Get bIsInFrontEnd()
   > Set isFrontend(isFrontend:double="To Float (Boolean)"(InBool=@n1)) -> (Output_Get:double)
   n2 = NOT Boolean(A=Get bIsPlayingEmote())
   > Set frontendPoseAlpa(frontendPoseAlpa:double="To Float (Boolean)"(InBool=AND Boolean(A=@n1, B=@n2))) -> (Output_Get:double)
   n3 = Event OnCharacterCustomizationLoaded()
   Delay(execute=[@n3.then, @n4.then, @n5.then], Duration=0.1)
   > ResetDynamics(InTeleportType=TeleportPhysics)
   > ResetDynamics(InTeleportType=ResetPhysics)
   n4 = Event OnBeginPlayingEmote() {is=title K2Node_Event}
   n5 = Event OnEndPlayingEmote() {is=title K2Node_Event}
   Event BlueprintInitializeAnimation()
   > Set Body_Mesh(Body_Mesh=GetOwnerSkeletalMeshComponent())
  ]}
]
variables [
 Body_Mesh {type=SkeletalMeshComponent}
 frontendPoseAlpa {type=number}
 isFrontend {type=number, default=0.1}
]
